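# Library material function that maps a 0-1 value to a spring-overshoot easing curve.
# A squared ease-in reaches 1 at a point derived from the offset and frequency inputs, then an amplitude-scaled sine oscillates around it, fading linearly as the value approaches 1.
# The value input previews with looping Time, and an unconnected SphereMask visualizes the curve as a moving dot.
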MaterialFunction /Game/Global/Materials/MaterialGym/Materials/MaterialElements/EaseCurves/MF_EaseCurves_SpringOvershoot
graph {
 properties {bExposeToLibrary=true}
 nodes [
  n1 = FunctionInput() {bUsePreviewValueAsDefault=true, inputName=Overshoot_Offset, inputType=FunctionInput_Scalar, sortPriority=1}
  n2 = FunctionInput() {bUsePreviewValueAsDefault=true, inputName=Overshoot_Frequency, inputType=FunctionInput_Scalar, sortPriority=2, previewValue=<5 0 0 1>}
  n3 = Divide(A=Add(A=Multiply(A=@n1) {constB=0.5}) {constB=0.5}, B=@n2) {constA=0.5}
  n4 = Saturate(None=FunctionInput(Preview=Time() {bOverride_Period=true, period=1}) {bUsePreviewValueAsDefault=true, inputName="Value (0-1)", inputType=FunctionInput_Scalar})
  n5 = Saturate(None=InverseLinearInterpolate(B=@n3, Value=@n4))
  n6 = OneMinus(None=@n4)
  n7 = InverseLinearInterpolate(A=@n3, Value=@n4)
  n8 = Sine(None=Multiply(A=Saturate(None=@n7), B=@n2) {constB=2})
  n9 = FunctionInput() {bUsePreviewValueAsDefault=true, inputName=Overshoot_Amplitude, inputType=FunctionInput_Scalar, sortPriority=3, previewValue=<0.1 0 0 1>}
  n10 = Add(A=Multiply(A=@n5, B=@n5), B=Multiply(A=Multiply(A=@n6, B=@n8), B=@n9) {constB=0.5})
  FunctionOutput(None=@n10) {outputName=Out}
  SphereMask(A=TextureCoordinate(), B=AppendVector(A=Constant() {r=0.5}, B=LinearInterpolate(Alpha=@n10) {constA=0.1, constB=0.9})) {attenuationRadius=0.1}
 ]
}
